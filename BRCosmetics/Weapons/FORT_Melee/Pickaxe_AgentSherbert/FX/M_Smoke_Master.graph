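# Master smoke material for the Agent Sherbert pickaxe FX: unlit, translucent and two-sided, for Niagara mesh particles.
# Opacity layers panning ring and smoke textures under a gradient mask, with depth and view-angle fades; mobile skips UV distortion.
# Particle color drives emissive and alpha.
# World position offset adds scrolling noise waves to a UV-weighted bend from clamped particle-relative offsets, controlled by Niagara dynamic parameters.
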Material /BRCosmetics/Weapons/FORT_Melee/Pickaxe_AgentSherbert/FX/M_Smoke_Master
graph {
 properties {
  bUsedWithNiagaraMeshParticles=true
  blendMode=BLEND_Translucent
  pixelDepthOffsetMode=PDOM_Legacy
  shadingModel=MSM_Unlit
  twoSided=true
  usedShadingModels=Unlit
 }
 nodes [
  n1 = DynamicParameter() {desc=Parameters set by Niagara, parameterIndex=1, paramNames [X,Y,Z,WPOIntensity]}
  n2 = Clamp(
   None=Subtract(
    A=MaterialFunctionCall(X=@n1.X, Y=@n1.X, Z=@n1.X) {materialFunction=&MakeFloat3}.Result
    B=ParticlePositionWS()
   )
   Min=ScalarParameter() {defaultValue=-100, parameterName=Clamp_min}
   Max=ScalarParameter() {defaultValue=100, parameterName=Clamp_max}
  ) {maxDefault=1000, minDefault=-1000}
  n3 = OneMinus(None=TextureCoordinate())
  n4 = Multiply(A=Power(Base=ComponentMask(None=@n3) {g=true}, Exp=ScalarParameter() {defaultValue=3, parameterName=WPOBendPower}), B=@n1.WPOIntensity)
  n5 = DynamicParameter() {defaultValue=<rgba 0.1 1 1 1>, paramNames [Offset, Texture_mask_offset, WaveIntensity, Param4]}
  n6 = NamedRerouteUsage()
  n7 = Panner(
   Coordinate=Add(A=Add(A=@n5.Offset, B=@n6), B=TextureCoordinate() {coordinateIndex=1})
   Time=Multiply(A=Time(), B=ScalarParameter() {defaultValue=1.1, group=Panner Speed, parameterName=Speed})
  ) {speedY=0.8}
  n8 = TextureSample(UVs=@n7) {samplerType=SAMPLERTYPE_Normal, texture=&T_Weap_Distort_N_02}
  n9 = TextureCoordinate()
  n10 = LinearInterpolate(
   B=AppendVector(
    A=ComponentMask(None=@n8.RGB) {g=true, r=true}
    B=TextureSample(UVs=@n7) {texture=&T_Smooth_tile}.R
   )
   Alpha=OneMinus(None=ComponentMask(None=@n9) {g=true})
  )
  n11 = DynamicParameter() {defaultValue=<rgba 0.1 1 1 1>, paramNames [Offset, Texture_mask_offset, WaveIntensity, Param4]}
  Add(
   A=LinearInterpolate(A=Constant(), B=LinearInterpolate(B=@n2, Alpha=@n4), Alpha=ScalarParameter() {defaultValue=1, parameterName=Use interpolate WPO})
   B=LinearInterpolate(
    A=Multiply(A=@n10, B=@n11.WaveIntensity) {constB=10}
    Alpha=ScalarParameter() {desc=no_Wpo, group=Simple_Round_particle, parameterName=No_WPO}
   )
  )
  n12 = DynamicParameter() {defaultValue=<rgba 0.1 1 1 1>, paramNames [Offset, Texture_mask_offset, WaveIntensity, Param4]}
  n13 = NamedRerouteUsage()
  n14 = Add(A=AppendVector(A=Constant(), B=Add(A=@n12.Offset, B=@n13)), B=TextureCoordinate() {vTiling=1.5})
  n15 = VectorParameter() {group=Fire Shape, parameterName=Fire_Micro_UVScale, sortPriority=8, defaultValue=<rgba 1 0.075 0 0>}
  n16 = Multiply(A=Add(A=@n14, B=TextureCoordinate() {uTiling=0.8, vTiling=0.9}), B=ComponentMask(None=@n15.RGB) {g=true, r=true})
  MaterialFunctionCall(XY=@n16) {materialFunction=&Swizzle}
  n17 = VertexColor()
  n18 = Constant()
  n19 = VectorParameter() {parameterName="RGBA (Offset)", defaultValue=<rgba 0.333 0.789 0.66 0.15>}
  n20 = Add(A=If(A=@n17.R, B=@n18, "A > B"=@n19.R, "A == B"=@n18, "A < B"=@n18), B=If(A=@n17.G, B=@n18, "A > B"=@n19.G, "A == B"=@n18, "A < B"=@n18))
  n21 = If(A=@n17.B, B=@n18, "A > B"=@n19.B, "A == B"=@n18, "A < B"=@n18)
  NamedRerouteDeclaration(None=Add(A=Add(A=@n20, B=@n21), B=If(A=@n17, B=@n18, "A > B"=@n18, "A == B"=@n19.A, "A < B"=@n18))) {name=VertexColorOutput, nodeColor=<rgba 0.2656 0 1 1>}
  n22 = ParticleColor()
  n23 = Constant3Vector()
  Reroute(None=MaterialFunctionCall(Emissive=Multiply(A=@n22.RGB, B=@n23)) {materialFunction=&MF_FX_EmissiveAdjustments}.Result)
  n24 = Constant2Vector() {g=0.5, r=0.75}
  n25 = Panner(Coordinate=@n16) {speedY=0.1}
  n26 = Add(A=ParticleRandom(), B=@n14)
  n27 = VectorParameter() {group=Distortion, parameterName=Distortion UV Scale, sortPriority=3, defaultValue=<rgba 1 0.05 0 0>}
  n28 = ComponentMask(None=@n27.RGB) {g=true, r=true}
  n29 = TextureSampleParameter2D(UVs=Panner(Coordinate=Multiply(A=@n26, B=@n28)) {speedY=0.2}) {
   group=Distortion
   parameterName=Distortion Normal
   samplerType=SAMPLERTYPE_Normal
   sortPriority=2
   texture=&T_Weap_Distort_N_02
  }
  n30 = VectorParameter() {group=Fire Shape, parameterName=Fire_MacroShape_UVScale, sortPriority=5, defaultValue=<rgba 1 1 0 0>}
  n31 = Multiply(A=TextureCoordinate(), B=ComponentMask(None=@n30.RGB) {g=true, r=true})
  n32 = Divide(
   A=LinearInterpolate(A=ComponentMask(None=@n29.RGB) {g=true, r=true}, Alpha=ComponentMask(None=@n31) {g=true}) {constB=0}
   B=ScalarParameter() {defaultValue=5, group=Distortion, parameterName=UV_Distortion_Divsor, sortPriority=1}
  ) {constB=20}
  n33 = ScalarParameter() {defaultValue=1, group=Distortion, parameterName=UV_Distortion_Lerp}
  n34 = ShadingPathSwitch(Default=Add(A=LinearInterpolate(B=@n32, Alpha=@n33), B=@n25) {constB=0}, Mobile=@n25) {desc=Migrated from FeatureLevelSwitch}
  n35 = TextureSample(UVs=Add(A=Constant2Vector() {r=0.2}, B=Multiply(A=@n24, B=@n34))) {
   samplerSource=SSM_Wrap_WorldGroupSettings
   samplerType=SAMPLERTYPE_LinearColor
   texture=/Game/Effects/Fort_Effects/Textures/Tile/T_AMR_MuzzleRing_01
  }
  n36 = TextureSample() {texture=/Game/Effects/Fort_Effects/Textures/Masks/T_Square_Gradient_M}
  n37 = Multiply(A=Multiply(A=@n35.G) {constB=0.5}, B=@n36.R)
  n38 = TextureCoordinate()
  n39 = ComponentMask(None=@n38) {r=true}
  n40 = OneMinus(None=@n39)
  n41 = Constant()
  n42 = Add(
   A=DynamicParameter() {defaultValue=<rgba 0.1 1 1 1>, paramNames [Offset, Texture_mask_offset, WaveIntensity, Param4]}.Offset
   B=NamedRerouteUsage()
  )
  n43 = Multiply(A=Add(A=@n38, B=AppendVector(A=@n41, B=@n42)), B=Constant2Vector() {g=0.5, r=1})
  n44 = TextureSampleParameter2D(UVs=Panner(Coordinate=MaterialFunctionCall(XY=@n43) {materialFunction=&Swizzle}.YX) {speedX=0.5}) {group=Smoke, parameterName=Smoke A Tex, texture=/Game/Effects/Fort_Effects/Textures/Smoke/T_Smoke_Trail_Soft_Chromatic}
  n45 = Add(
   A=Multiply(A=@n37, B=Multiply(A=@n40, B=@n39))
   B=MaterialFunctionCall(In=Multiply(A=@n44.G), Contrast=Constant() {r=0.25}) {materialFunction=&CheapContrast}.Result
  )
  n46 = ComponentMask(None=@n38) {g=true}
  n47 = Add(A=@n37, B=Saturate(None=Multiply(A=@n45, B=@n46)))
  n48 = Multiply(A=Power(Base=@n36.R)) {constB=1.5}
  n49 = DepthFade(Opacity=Multiply(A=Multiply(A=@n47, B=@n48)) {constB=2.5}) {fadeDistanceDefault=5}
  n50 = MaterialFunctionCall("FadeSensitivity (optional)"=Constant() {r=0.5}) {materialFunction=&FadeBasedOnViewAngle}
  Reroute(None=Multiply(A=@n22.A, B=Multiply(A=@n49, B=@n50.Result)))
  ScalarParameter() {defaultValue=30, group=WPO, parameterName=WaveIntensity, sortPriority=7}
 ]
}
